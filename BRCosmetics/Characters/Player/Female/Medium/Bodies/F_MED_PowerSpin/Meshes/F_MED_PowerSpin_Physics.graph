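# Physics asset for the female medium PowerSpin cosmetic body, rigging secondary motion for its T-shirt and sleeves.
# Spine capsules are locked in a chain, with zero-size base anchor bodies for the four shirt sides and both sleeves locked to spine_04.
# Shirt capsules and sleeve boxes hang from those anchors on tightly limited swing and twist constraints, allowing only slight movement.
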PhysicsAsset /BRCosmetics/Characters/Player/Female/Medium/Bodies/F_MED_PowerSpin/Meshes/F_MED_PowerSpin_Physics
bodies {
 Base_T_shirt_bk [{shapeType=Capsule, length=0, radius=0}]
 Base_T_shirt_f [{shapeType=Capsule, length=0, radius=0}]
 Base_T_shirt_l [{shapeType=Capsule, length=0, radius=0}]
 Base_T_shirt_r [{shapeType=Capsule, length=0, radius=0}]
 Base_sleeve_l [{shapeType=Capsule, length=0, radius=0}]
 Base_sleeve_r [{shapeType=Capsule, length=0, radius=0}]
 dyn_T_shirt_bk_2 [{shapeType=Capsule, length=8.514, radius=2.986, center=<2.8501 -2 0>}]
 dyn_T_shirt_f_2 [{shapeType=Capsule, length=8.41, radius=3.09, center=<4.2017 -2 0>}]
 dyn_T_shirt_l_2 [{shapeType=Capsule, length=8.41, radius=3.09, center=<3.6131 -2 0>}]
 dyn_T_shirt_r_2 [{shapeType=Capsule, length=8.41, radius=3.09, center=<2.1817 -2 0>}]
 sleeve_l [{shapeType=Box, extentX=5.9979, extentY=2.5071, extentZ=4.8352, center=<3.2197 0 0>}]
 sleeve_r [{shapeType=Box, extentX=10.2973, extentY=2.6113, extentZ=5.8147, center=<3.2939 0 0>}]
 spine_02 [{shapeType=Capsule, length=3.1088, radius=8.5829, shapeName=spine_02_capsule, center=<0 0.6014 -0.0169>, rotation=<rot 1.6075 -90 -0.6498>}]
 spine_03 [{shapeType=Capsule, length=2.5, radius=9.005, shapeName=spine_03_capsule, center=<0 1.5856 0>, rotation=<rot 0 -90 0>}]
 spine_04 [{shapeType=Capsule, length=3, radius=9.3403, shapeName=spine_04_capsule, center=<0 1.9515 -0.0986>, rotation=<rot 2.8927 -90 -3.3533>}]
}
constraints [
 {
  bone1Name=spine_03
  bone2Name=spine_02
  swing1LimitDegrees=45
  swing1Motion=Locked
  swing2LimitDegrees=45
  swing2Motion=Locked
  twistLimitDegrees=45
  twistMotion=Locked
 }
 {
  bone1Name=spine_04
  bone2Name=spine_03
  swing1LimitDegrees=45
  swing1Motion=Locked
  swing2LimitDegrees=45
  swing2Motion=Locked
  twistLimitDegrees=45
  twistMotion=Locked
 }
 {
  bone1Name=Base_T_shirt_bk
  bone2Name=spine_04
  swing1LimitDegrees=45
  swing1Motion=Locked
  swing2LimitDegrees=45
  swing2Motion=Locked
  twistLimitDegrees=45
  twistMotion=Locked
 }
 {bone1Name=dyn_T_shirt_bk_2, bone2Name=Base_T_shirt_bk, swing1LimitDegrees=20, swing2LimitDegrees=10, twistLimitDegrees=4}
 {
  bone1Name=Base_T_shirt_f
  bone2Name=spine_04
  swing1LimitDegrees=45
  swing1Motion=Locked
  swing2LimitDegrees=45
  swing2Motion=Locked
  twistLimitDegrees=45
  twistMotion=Locked
 }
 {bone1Name=dyn_T_shirt_f_2, bone2Name=Base_T_shirt_f, swing1LimitDegrees=20, swing2LimitDegrees=10, twistLimitDegrees=1}
 {
  bone1Name=Base_T_shirt_l
  bone2Name=spine_04
  swing1LimitDegrees=45
  swing1Motion=Locked
  swing2LimitDegrees=45
  swing2Motion=Locked
  twistLimitDegrees=45
  twistMotion=Locked
 }
 {bone1Name=dyn_T_shirt_l_2, bone2Name=Base_T_shirt_l, swing1LimitDegrees=20, swing2LimitDegrees=20, twistLimitDegrees=4}
 {
  bone1Name=Base_T_shirt_r
  bone2Name=spine_04
  swing1LimitDegrees=45
  swing1Motion=Locked
  swing2LimitDegrees=45
  swing2Motion=Locked
  twistLimitDegrees=45
  twistMotion=Locked
 }
 {bone1Name=dyn_T_shirt_r_2, bone2Name=Base_T_shirt_r, swing1LimitDegrees=20, swing2LimitDegrees=20, twistLimitDegrees=4}
 {bone1Name=sleeve_l, bone2Name=Base_sleeve_l, swing1LimitDegrees=10, swing2LimitDegrees=10, twistLimitDegrees=2}
 {bone1Name=sleeve_r, bone2Name=Base_sleeve_r, swing1LimitDegrees=10, swing2LimitDegrees=10, twistLimitDegrees=2}
 {
  bone1Name=Base_sleeve_r
  bone2Name=spine_04
  swing1LimitDegrees=45
  swing1Motion=Locked
  swing2LimitDegrees=45
  swing2Motion=Locked
  twistLimitDegrees=45
  twistMotion=Locked
 }
 {
  bone1Name=Base_sleeve_l
  bone2Name=spine_04
  swing1LimitDegrees=45
  swing1Motion=Locked
  swing2LimitDegrees=45
  swing2Motion=Locked
  twistLimitDegrees=45
  twistMotion=Locked
 }
]
