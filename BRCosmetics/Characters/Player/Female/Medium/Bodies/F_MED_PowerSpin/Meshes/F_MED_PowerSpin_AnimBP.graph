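# Animation Blueprint for the F_MED_PowerSpin female medium body cosmetic, following the attached parent mesh and adding secondary motion.
# The copied pose runs through a Control Rig fed by a smoothed emote value, then AnimDynamics on the Kit bone and a pelvis-based RigidBody simulation.
# Two other RigidBody nodes are unconnected.
# The event graph sets GravityOverride using PartyJelly's parameter asset, but the AnimGraph never reads it.
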AnimBlueprint /BRCosmetics/Characters/Player/Female/Medium/Bodies/F_MED_PowerSpin/Meshes/F_MED_PowerSpin_AnimBP
target_skeleton=&/F_MED_PowerSpin_Skeleton
graphs [
 AnimGraph {nodes [
   RigidBody() {
    bClampLinearTranslationLimitToRefPose=true
    bOverrideWorldGravity=true
    overridePhysicsAsset=&/F_MED_PowerSpin_Physics
    simulationSpace=BaseBoneSpace
    alphaBoolBlend {blendOption=Linear}
    baseBoneRef {boneName=spine_03}
    componentAppliedLinearAccClamp=<3500 3500 3500>
    overrideWorldGravity {z=-980}
    simSpaceSettings {maxAngularAcceleration=100, maxAngularVelocity=5, maxLinearAcceleration=100, maxLinearVelocity=150, worldAlpha=1}
   }
   RigidBody!(ComponentPose, Alpha, OverrideWorldGravity:Vector="0,0,0", ExternalForce) -> (Pose) {
    bOverrideWorldGravity=true
    overridePhysicsAsset=&F_MED_PowerSpin_Physics
    simulationSpace=BaseBoneSpace
    alphaBoolBlend {blendOption=Linear}
    baseBoneRef {boneName=pelvis}
    overrideWorldGravity {z=-980}
    simSpaceSettings {maxAngularAcceleration=80, maxAngularVelocity=20, maxLinearAcceleration=3500, maxLinearVelocity=700, velocityScaleZ=0.4, worldAlpha=1}
   }
   n1 = Control Rig!(
    Source=Copy Pose From Mesh() {bUseAttachedParent=true}
    Alpha
    Emote:bool=Get bIsPlayingEmote()
    Emote_Float:double=Get Emote() -> (Emote:double)
   ) -> (Pose) {controlRigAssetReference {blueprintRigClass=&/F_MED_PowerSpin_CtrlRig.F_MED_PowerSpin_CtrlRig_C}}
   n2 = Anim Dynamics - Bone: Kit(ComponentPose=Local To Component(LocalPose=@n1), Alpha=1, ExternalForce="0,0,0") {
    angularDampingOverride=1
    angularSpringConstant=500
    bAngularSpring=true
    bOverrideAngularDamping=true
    bUseGravityOverride=true
    simulationSpace=RootRelative
    alphaBoolBlend {blendOption=Linear}
    boundBone {boneName=Kit}
    componentAppliedLinearAccClamp=<5000 5000 5000>
    componentLinearAccScale=<1 1 1>
    componentLinearVelScale=<1 1 1>
    gravityOverride {z=-980}
    physicsBodyDefinitions [{
      boundBone {boneName=Kit}
      boxExtents=<10 10 10>
      constraintSetup {angularLimitsMax=<5 2 0>, angularLimitsMin=<-5 -6 0>, angularTarget=<1 0 0>}
      localJointOffset=<0 0 0.1>
     }]
    relativeSpaceBone {boneName=head}
   }
   n3 = RigidBody(ComponentPose=@n2, OverrideWorldGravity="0,0,0") {
    bOverrideWorldGravity=true
    overridePhysicsAsset=&F_MED_PowerSpin_Physics
    simulationSpace=BaseBoneSpace
    alphaBoolBlend {blendOption=Linear}
    baseBoneRef {boneName=pelvis}
    overrideWorldGravity {z=-980}
    simSpaceSettings {
     dampingAlpha=0
     maxAngularAcceleration=500
     maxAngularVelocity=50
     maxLinearAcceleration=600
     maxLinearVelocity=250
     worldAlpha=1
     externalLinearDragV=<0.1 0.1 0.5>
    }
   }
   Output Pose(Result=Component To Local(ComponentPose=@n3)) {layerGroup=None}
  ]}
 EventGraph {nodes [
   n1 = Event BlueprintUpdateAnimation()
   > Set Delta Time X(Delta Time X=@n1.DeltaTimeX)
   > n2 = Branch(Condition=Get bIsPlayingEmote())
   > Set Emote(Emote:double=FInterpTo(Current=Get Emote(), Target=1, DeltaTime=Get Delta Time X(), InterpSpeed=10)) -> (Output_Get:double)
   Set Emote(execute=@n2.else, Emote=FInterpTo(Current=Get Emote(), Target=0, DeltaTime=Get Delta Time X(), InterpSpeed=10))
   > n3 = "Gravity Override State Function (Parameter Asset) (Nativized)"(
    CharacterPartAnimInstance=Self-Reference()
    GameTimeInSeconds=GetGameTimeInSeconds()
    Parameters=/BRCosmetics/Characters/Player/Female/Medium/Bodies/F_MED_PartyJelly/Meshes/M_MED_PartyJelly_Parameters
   )
   > Set GravityOverride(GravityOverride=@n3)
  ]}
]
variables [
 Delta Time X {type=number}
 Emote {type=number}
 GravityOverride {type=Vector}
]
